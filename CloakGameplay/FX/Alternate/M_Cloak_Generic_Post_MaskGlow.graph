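# Post-process material for the cloak effects that adds a colored glow around the screen edges, blue by default.
# An inverted, soft SphereMask centered on the screen UVs forms the vignette, tinted by the Color parameter.
# The Alpha value from the MPC_Cloak_Generic_Post_MaskGlow parameter collection scales the glow globally, and the result is added to the PostProcessInput0 scene color.
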Material /CloakGameplay/FX/Alternate/M_Cloak_Generic_Post_MaskGlow
graph {
 properties {materialDomain=MD_PostProcess, pixelDepthOffsetMode=PDOM_Legacy, shadingModel=MSM_Unlit, usedShadingModels=Unlit}
 nodes [
  n1 = VectorParameter() {parameterName=Color, defaultValue=<rgba 0.14 0.2333 0.7 1>}
  n2 = OneMinus(None=SphereMask(A=TextureCoordinate(), B=Constant2Vector() {g=0.5, r=0.5}, Radius=Constant() {r=0.75}, Hardness=Constant() {r=0.5}))
  Add(
   A=Multiply(
    A=Multiply(A=@n1.RGB, B=@n2)
    B=CollectionParameter() {collection=&/MPC_Cloak_Generic_Post_MaskGlow, parameterName=Alpha}
   )
   B=ComponentMask(None=SceneTexture() {sceneTextureId=PPI_PostProcessInput0}.Color) {b=true, g=true, r=true}
  )
 ]
}
